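; BP_StageScreenActor is a stage screen that receives material and Niagara effect changes, identified by its ScreenId or display name.
; EventUpdateMaterial fades out through custom primitive data stamped with game time, swaps materials by slot name after a delay, then fades in; fading is optional.
; UpdateNiagara replaces, creates or destroys the particle component as needed.
; The construction script applies the configured mesh, or without one rebuilds the configured particle component.
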(event EventUpdateMaterial
  (params (MaterialSlotName :type Array%20of%20Names) (Material :type Array%20of%20Material%20Instance%20Object%20References))
  (bind _screenmesh (Variables|Default|GetScreenMesh)
    (:then
      (Rendering|Material|SetVectorParameterforCustomPrimitiveData _screenmesh "ScreenFadePrimitiveData" (Math|Vector4|MakeVector4 0.0 (Utilities|Time|GetGameTimeinSeconds) 2.3 (select (Variables|Default|GetUseFadeInOut) 1.0 0.0))) ; Value X -> 0 To fade Out 1 to Fade In\nValue Z -> Fade Speed\nValue W -> SkipFade 0 Skip - 1 Fade
      (Utilities|FlowControl|Delay (select (Variables|Default|GetUseFadeInOut) 0.75 0.0))
      (for (_array_element _array_index) MaterialSlotName
        (Utilities|IsValid (Utilities|Array|Get(acopy) Material _array_index)
          (:Is%20Valid
            (Rendering|Material|SetMaterialbyName (Variables|Default|GetScreenMesh) _array_element (Utilities|Array|Get(acopy) Material _array_index)))))
      (Rendering|Material|SetVectorParameterforCustomPrimitiveData (Variables|Default|GetScreenMesh) "ScreenFadePrimitiveData" (Math|Vector4|MakeVector4 1.0 (Utilities|Time|GetGameTimeinSeconds) 1.7 (select (Variables|Default|GetUseFadeInOut) 1.0 0.0))) ; Value X -> 0 To fade Out 1 to Fade In\nValue Z -> Fade Speed\nValue W -> SkipFade 0 Skip - 1 Fade
    )
    (:else)))

(fn GetMaterialsSlotNames
  (return (Rendering|Material|GetMaterialSlotNames (Variables|Default|GetScreenMesh))))

(fn GetScreenID
  (if (== (Variables|Default|GetScreenId) "")
    (return (Utilities|String|StringToName (Utilities|GetDisplayName self)))
    (else
      (return (Variables|Default|GetScreenId)))))

(fn UpdateNiagara
  (params (Niagara :type Niagara%20System%20Object%20Reference))
  (bind _output_get (LocalVariables|SetLocalNiagara Niagara))
  (bind _particlesystemcomponent (Variables|Default|GetParticleSystemComponent)
    (:then
      ;[0.3211 0.361 1] If there is a component and the received niagara is valid them just replace to another niagara.
      (Utilities|IsValid _output_get
        (:Is%20Valid
          (Niagara|SetNiagaraSystemAsset _particlesystemcomponent (LocalVariables|GetLocalNiagara) true)
          (return true))
        (:Is%20Not%20Valid
          ;[1 0.0816 0.0816] Destroy the Niagara Component in case the niagara receive is none.
          (bind _particlesystemcomponent_2 (Variables|Default|GetParticleSystemComponent)
            (:then
              (Components|DestroyComponent _particlesystemcomponent_2)
              (return false))
            (:else)))))
    (:else
      ;[0.1954 1 0.2192] Create a new niagara component in case there is none and we do have a niagara to set.
      (Utilities|IsValid _output_get
        (:Is%20Valid
          (bind _returnvalue (AddComponent|NODE_AddNiagaraComponent-2))
          (bind _output_get_2 (Variables|Default|SetParticleSystemComponent _returnvalue))
          (Niagara|SetNiagaraSystemAsset _output_get_2 (LocalVariables|GetLocalNiagara) true)
          (return true))))))

(fn ConstructionScript
  (Utilities|IsValid (Variables|Default|GetStageScreenMesh)
    (:Is%20Valid
      (if (!= (Variables|/Script/Engine.StaticMeshComponent|GetStaticMesh (Variables|Default|GetScreenMesh)) (Variables|Default|GetStageScreenMesh))
        (Components|StaticMesh|SetStaticMesh (Variables|Default|GetScreenMesh) (Variables|Default|GetStageScreenMesh))))
    (:Is%20Not%20Valid
      (bind _particlesystemcomponent (Variables|Default|GetParticleSystemComponent)
        (:then
          (Components|DestroyComponent _particlesystemcomponent))
        (:else
          _))
      (Variables|Default|GetStageScreenParticle
        (:then
          (bind _returnvalue (AddComponent|NODE_AddNiagaraComponent-1))
          (bind _output_get (Variables|Default|SetParticleSystemComponent _returnvalue))
          (Niagara|SetNiagaraSystemAsset _output_get (Variables|Default|GetStageScreenParticle) true))
        (:else
          (bind _particlesystemcomponent_2 (Variables|Default|GetParticleSystemComponent)
            (:then
              (Components|DestroyComponent _particlesystemcomponent_2))
            (:else)))))))
