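; MojoStageScreenEventInterface is an interface for querying stage screen meshes, materials and Niagara effects within the Sparks stage show's screen manager.
; It declares getters for mesh keys, values and legacy values, plus Niagara keys, values and lookup by key.
; Material queries take a screen stage key, slot-name arrays, or a material type structure with a screen type name.
; Every function returns only default values, leaving the logic to implementers.
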(fn GetFilteredMaterialsByType
  (params (SlotNames :type Array%20of%20Names))
  (return 0 false))

(fn GetMaterialsAndSlots
  (params (ScreenStageKey :type Name))
  (return 0 0 false))

(fn GetMaterialsAndSlotsByType
  (params (MaterialSlotNames :type Array%20of%20Names))
  (return 0 0 false))

(fn GetScreenMaterialType
  (params (Materials :type S%20Mojo%20Stage%20Screen%20Material%20Type%20Structure) (ScreenType :type Name))
  (return 0))

(fn GetScreenMeshesKey
  (return 0))

(fn GetScreenMeshesLegacyValues
  (return 0))

(fn GetScreenMeshesValues
  (return 0))

(fn GetScreenNiagara
  (params (Key :type Name))
  (return 0))

(fn GetScreenNiagaraKeys
  (return 0))

(fn GetScreenNiagaraValues
  (return 0))
