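; Blueprint interface for stage screen show effects in the Sparks stage show's screen manager system.
; It declares functions to query a screen's ID and material slot names, apply material instances to named slots, and set a Niagara system, returning a success flag.
; All bodies are empty defaults, so each implementing screen supplies the behaviour.
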(fn GetMaterialsSlotNames
  (return 0))

(fn GetScreenID
  (return "None"))

(fn UpdateMaterial
  (params (MaterialSlotName :type Array%20of%20Names) (Material :type Array%20of%20Material%20Instance%20Object%20References))
  (return))

(fn UpdateNiagara
  (params (Niagara :type Niagara%20System%20Object%20Reference))
  (return false))
